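# Translucent unlit sky dome material that overlays a color gradient following time of day, recentered on the camera and offset along vertex normals.
# Opacity rises from the horizon to the zenith, scaled by per-time-of-day values and softened by depth fade.
# A vertex-interpolated exponent widens the gradient as the camera climbs relative to a cloud altitude.
# Epic quality blends four time-of-day colors, while other levels use one flat color.
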Material /Game/TimeOfDay/TODM/BR/s11/Janus/M_SkyGradDome
graph {
 properties {bUseMaterialAttributes=true, blendMode=BLEND_Translucent, pixelDepthOffsetMode=PDOM_Legacy, shadingModel=MSM_Unlit, usedShadingModels=Unlit}
 nodes [
  n1 = CameraPositionWS()
  n2 = ObjectPositionWS()
  n3 = VertexNormalWS()
  n4 = ScalarParameter() {defaultValue=-500000, parameterName=NormalOffset}
  n5 = VectorParameter() {parameterName=Color, defaultValue=<rgba 0.0001 0.0145 0.0312 1>}
  n6 = MaterialFunctionCall(
   Color_Morning=VectorParameter() {parameterName=EpicColorMorning, defaultValue=<rgba 0.008 0.0024 0.0312 1>}.RGB
   Color_Day=VectorParameter() {parameterName=EpicColorDay, defaultValue=<rgba 0.0001 0.0145 0.0312 1>}.RGB
   Color_Evening=VectorParameter() {parameterName=EpicColorEvening, defaultValue=<rgba 0.0184 0.0005 0.0312 1>}.RGB
   Color_Night=VectorParameter() {parameterName=EpicColorNight, defaultValue=<rgba 0.0231 0.0036 0.0312 1>}.RGB
  ) {materialFunction=&Color_Time_of_Day}
  n7 = CameraPositionWS()
  n8 = WorldPosition()
  n9 = MaterialFunctionCall(Float3=Normalize(VectorInput=Subtract(A=@n7, B=@n8.XYZ))) {materialFunction=&BreakOutFloat3Components}
  n10 = OneMinus(None=Clamp(None=Multiply(A=@n9.B) {constB=-1}))
  n11 = Constant() {r=160000}
  n12 = Add(A=Constant() {desc=CLOUD ALTITUDE, r=35662.3828}, B=@n11)
  n13 = CameraPositionWS()
  n14 = SphereMask(
   A=@n12
   B=Min(
    A=@n12
    B=MaterialFunctionCall(Float3=@n13) {materialFunction=&BreakOutFloat3Components}.B
   )
   Radius=@n11
   Hardness=Constant()
  )
  n15 = Constant() {r=0.3}
  n16 = VertexInterpolator(VS=LinearInterpolate(A=Constant() {r=30}, B=Constant() {r=2}, Alpha=Power(Base=@n14, Exp=@n15) {constExponent=0.2}) {constA=100})
  n17 = Multiply(
   A=OneMinus(None=Power(Base=@n10, Exp=@n16.PS))
   B=MaterialFunctionCall(
    Value_Morning=ScalarParameter() {defaultValue=0.6, parameterName=opacityMorning}
    Value_Day=ScalarParameter() {defaultValue=0.4, parameterName=opacityDay}
    Value_Evening=ScalarParameter() {defaultValue=0.5, parameterName=opacityEvening}
    Value_Night=ScalarParameter() {defaultValue=0.55, parameterName=opacityNight}
   ) {materialFunction=&Scalar_Time_of_Day}.Result
  )
  n18 = ScalarParameter() {defaultValue=500000, parameterName=DepthFade}
  SetMaterialAttributes!(
   MaterialAttributes=SetMaterialAttributes!(MaterialAttributes, Emissive Color=QualitySwitch(Default=@n5.RGB, Epic=@n6.Result), Opacity=DepthFade(Opacity=@n17, FadeDistance=@n18)) {attributeSetTypes [#1, #2]}
   World Position Offset=Add(A=Subtract(A=@n1, B=@n2), B=Multiply(A=@n3, B=@n4) {constB=-500000})
  ) {attributeSetTypes [#3]}
 ]
}
